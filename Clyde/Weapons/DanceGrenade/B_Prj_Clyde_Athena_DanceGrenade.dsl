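; Dance grenade projectile Blueprint for the Clyde Athena weapon, adding an obstruction test against a target pawn.
; CheckObstructed line-traces on TraceTypeQuery6 from the grenade's location to the target's location.
; On a hit, it returns whether the hit actor has any tag in ObstructorTags; with no hit it returns false.
; The construction script only calls the parent, and no other logic appears here.
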(fn CheckObstructed
  (params (Target :type Fort%20Pawn%20Object%20Reference))
  (bind _outhit (Collision|LineTraceByChannel (Transformation|GetActorLocation) (Transformation|GetActorLocation Target) "TraceTypeQuery6"))
  (if (Collision|BreakHitResult _outhit)
    (bind _returnvalue (GameplayTags|HasAnyMatchingGameplayTags(Message) (.HitActor (Collision|BreakHitResult _outhit)) (Variables|Default|GetObstructorTags)))
    (return _returnvalue)
    (else
      (return false))))

(fn ConstructionScript
  (|CallParentFunction))
