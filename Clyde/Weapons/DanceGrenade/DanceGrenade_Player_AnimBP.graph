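# Player anim Blueprint for the Dance Grenade that layers curve-driven hip sway and bounce onto the character through the ItemPreIK layer.
# Each update reads animation curves for bounce, sway and pelvis pitch, and interpolates MovingAlpha toward 1 or 0 based on bIsMoving2D.
# ItemPreIK additively offsets the pelvis, then counter-pitches Spine_01 and Spine_02 by half each.
# MovingAlpha weights pelvis and Spine_01, while Spine_02 stays at full alpha.
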AnimBlueprint /Clyde/Weapons/DanceGrenade/DanceGrenade_Player_AnimBP
target_skeleton=&Fortnite_M_Avg_Player_Skeleton
graphs [
 AnimGraph {nodes [Output Pose()]}
 EventGraph {nodes [
   Event BlueprintUpdateAnimation()
   > Set VerticalBounce(VerticalBounce:double=GetCurveValue(CurveName=UserCurve01)) -> (Output_Get:double)
   > Set SideToSide(SideToSide:float=float * float(A=GetCurveValue(CurveName=UserCurve02), B=-1) {is=title K2Node_PromotableOperator}) -> (Output_Get:float)
   > n1 = Branch(Condition=Get bIsMoving2D())
   > n2 = Set MovingAlphaTarget(MovingAlphaTarget:double=1) -> (Output_Get:double)
   Set MovingAlpha(
    execute=[@n4.then, @n2.then]
    MovingAlpha:double=FInterpTo_Constant!(
     Current:double=Get MovingAlpha() -> (MovingAlpha:double)
     Target:double=Get MovingAlphaTarget() -> (MovingAlphaTarget:double)
     DeltaTime:double=Get DeltaTime() -> (DeltaTime:float)
     InterpSpeed:double=3
    ) -> (ReturnValue:double)
   ) -> (Output_Get:double)
   > Set HipOffsetPitch(HipOffsetPitch:float=GetCurveValue(CurveName=PelvisRotationOffset)) -> (Output_Get:float)
   n3 = Get HipOffsetPitch() -> (HipOffsetPitch:float)
   > Set HipOffsetPitchOpp(HipOffsetPitchOpp:double=float / float(A=@n3, B=-2)) -> (Output_Get:double)
   n4 = Set MovingAlphaTarget(execute=@n1.else, MovingAlphaTarget=0)
   TryGetPawnOwner()
   float * float(A=@n3, B=-1)
  ]}
 ItemAimOffset {nodes [
   Input Pose!() -> (Pose, AimOffsetAlpha:double, Yaw:double, Pitch:double) {name=InPose_AimOffset}
   Input Pose!() -> (Pose) {name=InPose_PreAimOffsets}
   Output Pose() {layerGroup=Item Override}
  ]}
 ItemFallAdditive {nodes [
   Input Pose() {name=InPoseItemFallAdditive}
   Output Pose() {layerGroup=Item Override}
  ]}
 ItemFallLandAdditive {nodes [
   Input Pose() {name=InPoseItemFallLandAdditive}
   Output Pose() {layerGroup=Item Override}
  ]}
 ItemFallLevel2Override {nodes [
   Input Pose() {name=InPose_FallLevel2}
   Output Pose() {layerGroup=Item Override}
  ]}
 ItemFinalPoseOverride {nodes [
   Input Pose() {name=InFinalPose}
   Output Pose() {layerGroup=Item Override}
  ]}
 ItemFullBodyAimOffsets {nodes [
   Input Pose!() -> (Pose, FullBody_AimOffsetAlpha:double, Aim_Yaw:double, Aim_Pitch:double) {name=InPoseFullBodyAimOffsets}
   Input Pose!() -> (Pose) {name=InPosePreFullBodyAimOffsets}
   Output Pose() {layerGroup=Item Override}
  ]}
 ItemFullBodyOverride {nodes [
   Input Pose() {name=InPoseFullBody}
   Output Pose() {layerGroup=Item Override}
  ]}
 ItemIdleOverride {nodes [
   Input Pose() {name=InPose_Idle}
   Output Pose() {layerGroup=Item Override}
  ]}
 ItemJetPackJumpAdditive {nodes [Output Pose() {layerGroup=Item Override}]}
 ItemJetPackStartAdditive {nodes [Output Pose() {layerGroup=Item Override}]}
 ItemJumpLoopAdditive {nodes [
   Input Pose() {name=InPoseItemJumpLoopAdditive}
   Output Pose() {layerGroup=Item Override}
  ]}
 ItemJumpOverride {nodes [
   Input Pose() {name=InPoseJump}
   Output Pose() {layerGroup=Item Override}
  ]}
 ItemJumpUpAdditive {nodes [
   Input Pose() {name=InPoseItemJumpUpAdditive}
   Output Pose() {layerGroup=Item Override}
  ]}
 ItemPreAimOffsets {nodes [
   Input Pose() {name=InPosePreAimOffsets}
   Output Pose() {layerGroup=Item Override}
  ]}
 ItemPreIK {nodes [
   n1 = "Transform (Modify) Bone - Bone: pelvis"!(
    ComponentPose=Local To Component(LocalPose=Use cached pose 'FinalPoseCache'())
    Alpha=Get MovingAlpha()
    Translation_X:double=Get SideToSide() -> (SideToSide:float)
    Translation_Y:double=0
    Translation_Z:double=Get VerticalBounce() -> (VerticalBounce:double)
    Rotation_Roll:float=0
    Rotation_Pitch:float=Get HipOffsetPitch()
    Rotation_Yaw:float=0
    Scale="1,1,1"
   ) -> (Pose) {
    rotationMode=BMM_Additive
    rotationSpace=BCS_BoneSpace
    translationMode=BMM_Additive
    alphaBoolBlend {blendOption=Linear}
    boneToModify {boneName=pelvis}
   }
   n2 = Get HipOffsetPitchOpp() -> (HipOffsetPitchOpp:double)
   n3 = "Transform (Modify) Bone - Bone: Spine_01"!(
    ComponentPose=@n1
    Alpha=Get MovingAlpha()
    Translation="0,0,0"
    Rotation_Roll:float=0
    Rotation_Pitch:float=@n2
    Rotation_Yaw:float=0
    Scale="1,1,1"
   ) -> (Pose) {rotationMode=BMM_Additive, rotationSpace=BCS_BoneSpace, alphaBoolBlend {blendOption=Linear}, boneToModify {boneName=Spine_01}}
   n4 = "Transform (Modify) Bone - Bone: Spine_02"!(ComponentPose=@n3, Alpha=1, Translation="0,0,0", Rotation_Roll:float=0, Rotation_Pitch:float=@n2, Rotation_Yaw:float=0, Scale="1,1,1") -> (Pose) {rotationMode=BMM_Additive, rotationSpace=BCS_BoneSpace, alphaBoolBlend {blendOption=Linear}, boneToModify {boneName=Spine_02}}
   Output Pose(Result=Component To Local(ComponentPose=@n4)) {layerGroup=Item Override}
   Save cached pose 'FinalPoseCache'(Pose=Input Pose() {name=InPose_PreIK}) {cacheName=FinalPoseCache}
  ]}
 ItemRightHandIK {nodes [
   Input Pose() {name=InPosePreRightHandIK}
   Input Pose() {name=InPosePostRightHandIK}
   Output Pose() {layerGroup=Item Override}
  ]}
 ItemSkeletalControl {nodes [
   Input Pose()
   Output Pose() {layerGroup=Item Override}
  ]}
 ItemSwimJumpFallLoop {nodes [Output Pose() {layerGroup=Item Override}]}
 ItemSwimJumpSurfaceEnd {nodes [Output Pose() {layerGroup=Item Override}]}
 ItemSwimJumpSurfaceLoop {nodes [Output Pose() {layerGroup=Item Override}]}
 ItemUpperBody {nodes [
   Input Pose!() -> (Pose, InputParam:Fort Anim Input Adjusted Aim Structure) {name=InPoseUpperBody}
   Output Pose() {layerGroup=Item Override}
  ]}
 ItemZipLineStartAdditive {nodes [Output Pose() {layerGroup=Item Override}]}
]
variables [
 HipOffsetPitch {type=number}
 HipOffsetPitchOpp {type=number}
 MovingAlpha {type=number}
 MovingAlphaTarget {type=number}
 SideToSide {type=number}
 VerticalBounce {type=number}
]
