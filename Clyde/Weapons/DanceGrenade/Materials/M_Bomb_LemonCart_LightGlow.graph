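# Translucent two-sided glow material for the DanceGrenade's LemonCart bomb lights, built on the MF_LemonCart_Bomb attributes.
# Opacity masks non-black base color with a radial gradient, fades as vertex normals align with the view direction, and applies depth fade.
# The same view-angle term blends out Sprite billboarding and camera offset in World Position Offset, and blends normals back to vertex normals.
# Time of day scales emissive.
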Material /Clyde/Weapons/DanceGrenade/Materials/M_Bomb_LemonCart_LightGlow
graph {
 properties {bAutomaticallySetUsageInEditor=false, bUseMaterialAttributes=true, blendMode=BLEND_Translucent, twoSided=true, usedShadingModels=Default Lit}
 nodes [
  n1 = Constant3Vector() {constant=<rgba 0 0 1 0>}
  LinearInterpolate(
   A=ScalarParameter() {defaultValue=1, parameterName=Opacity.Fresnel_B.Value}
   B=ScalarParameter() {parameterName=Opacity.Fresnel_A.Value}
   Alpha=MaterialFunctionCall(
    Brightness=ScalarParameter() {defaultValue=1, parameterName=Opacity.Fresnel.Brightness}
    Exponent=ScalarParameter() {defaultValue=1, parameterName=Opacity.Fresnel.Exponent}
    Vertex / Pixel Normal=Transform(None=@n1)
   ) {materialFunction=&MF_Fresnel_FX}.Fresnel
  )
  n2 = MaterialFunctionCall(Offset Amount=ScalarParameter() {parameterName=CameraOffset}) {materialFunction=/Engine/Functions/Engine_MaterialFunctions02/WorldPositionOffset/CameraOffset}
  n3 = MaterialFunctionCall!(
   Center Location=WorldPosition().XYZ
   0-1 UVs=TextureCoordinate() {coordinateIndex=1}
   XY scale=ScalarParameter() {defaultValue=1, parameterName=Scale}
   OverridePivot
   "Normals (Optional)"
  ) {materialFunction=&Sprite}
  n4 = Subtract(A=CameraPositionWS(), B=ObjectPositionWS())
  n5 = Abs(None=DotProduct(A=Normalize(VectorInput=@n4), B=VertexNormalWS()))
  n6 = ScalarParameter() {defaultValue=1, parameterName=WPO.Fresnel.Power}
  n7 = Saturate(None=Multiply(A=Power(Base=@n5, Exp=@n6), B=ScalarParameter() {defaultValue=1, parameterName=WPO.Fresnel.Multi}))
  n8 = VertexNormalWS()
  n9 = MaterialFunctionCall!(Material, Alpha=Constant() {r=1}) {materialFunction=&/MF_LemonCart_Bomb}
  n10 = GetMaterialAttributes(None=@n9.Result) {attributeGetTypes [#1]}
  n11 = Constant()
  n12 = If(
   A=Divide(A=MaterialFunctionCall(f3=@n10.Base Color) {materialFunction=&AddComponents}.f3) {constB=3}
   "A > B"=Constant() {r=1}
   "A == B"=@n11
   "A < B"=@n11
  ) {constB=0.01}
  n13 = MaterialFunctionCall(
   UVs=TextureCoordinate() {coordinateIndex=1}
   Radius=ScalarParameter() {defaultValue=0.5, parameterName=Opacity.RadialGlow.Radius}
   Density=ScalarParameter() {defaultValue=2.3, parameterName=Opacity.RadialGlow.Density}
  ) {materialFunction=&RadialGradientExponential}
  n14 = Subtract(A=CameraPositionWS(), B=ObjectPositionWS())
  n15 = Abs(None=DotProduct(A=Normalize(VectorInput=@n14), B=VertexNormalWS()))
  n16 = ScalarParameter() {defaultValue=1, parameterName=WPO.Fresnel.Power}
  n17 = Saturate(None=Multiply(A=Power(Base=@n15, Exp=@n16), B=ScalarParameter() {defaultValue=1, parameterName=WPO.Fresnel.Multi}))
  n18 = Multiply(
   A=Multiply(A=Multiply(A=@n12, B=@n13.RadialGradientExponential), B=OneMinus(None=@n17))
   B=ScalarParameter() {defaultValue=1, parameterName=Opacity.Master}
  )
  n19 = ScalarParameter() {defaultValue=5, parameterName=Opacity.DepthFade}
  n20 = GetMaterialAttributes(None=SetMaterialAttributes!(MaterialAttributes=@n9.Result, Opacity=DepthFade(Opacity=@n18, FadeDistance=@n19) {fadeDistanceDefault=5}) {attributeSetTypes [#2]}) {attributeGetTypes [#3]}
  n21 = ScalarParameter() {defaultValue=1, parameterName=TOD.Emissive.Multi.Day}
  n22 = ScalarParameter() {defaultValue=1, parameterName=TOD.Emissive.Multi.Night}
  n23 = LinearInterpolate(A=@n21, B=@n22)
  n24 = MaterialFunctionCall(Value_Morning=@n23, Value_Day=@n21, Value_Evening=@n23, Value_Night=@n22) {materialFunction=&Scalar_Time_of_Day}
  SetMaterialAttributes!(
   MaterialAttributes=SetMaterialAttributes!(MaterialAttributes=@n20.MaterialAttributes, Emissive Color=Multiply(A=@n20.Emissive Color, B=@n24.Result)) {attributeSetTypes [#3]}
   World Position Offset=LinearInterpolate(A=Add(A=@n3.World position offset, B=@n2.Camera Offset), Alpha=@n7) {constB=0}
   Normal=Normalize(VectorInput=LinearInterpolate(A=@n3.Normals, B=@n8, Alpha=@n7) {constB=0})
  ) {attributeSetTypes [#4, #5]}
  n25 = GetMaterialAttributes() {attributeGetTypes [#2]}
  SetMaterialAttributes!(
   MaterialAttributes=@n25.MaterialAttributes
   Base Color=@n25.MaterialAttributes
   Emissive Color=@n25.MaterialAttributes
   Opacity=Constant() {r=1}
  ) {attributeSetTypes [#1, #3, #2]}
  n26 = TextureCoordinate() {coordinateIndex=1}
  n27 = ComponentMask(None=@n26) {g=true}
  n28 = Subtract(A=AppendVector(A=ComponentMask(None=@n26) {r=true}, B=OneMinus(None=@n27))) {constB=0.5}
  n29 = Max(A=SquareRoot(None=DotProduct(A=@n28, B=@n28))) {constB=0}
  n30 = WorldPosition()
  n31 = ObjectPositionWS()
  n32 = AppendVector(A=Multiply(A=Divide(A=@n28, B=@n29), B=Distance(A=@n30.XYZ, B=@n31)), B=Constant())
  Subtract(A=Add(A=Transform(None=@n32) {transformSourceType=TRANSFORMSOURCE_Camera}, B=ObjectPositionWS()), B=WorldPosition().XYZ)
 ]
}
