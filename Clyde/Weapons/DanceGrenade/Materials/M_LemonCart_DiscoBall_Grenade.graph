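# Lit surface material for the LemonCart disco ball dance grenade, built on the shared gun master shader and weapon skin FX functions.
# World Position Offset sums a pickup LOD effect and a black hole function, which samples the LOD-shifted position.
# MF_LemonCart_Bomb then processes the attributes, taking the SkinFX mask's red channel as alpha.
# The result passes through shared Hologram and emissive adjustment functions.
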Material /Clyde/Weapons/DanceGrenade/Materials/M_LemonCart_DiscoBall_Grenade
graph {
 properties {
  bUseMaterialAttributes=true
  bUsedWithInstancedStaticMeshes=true
  bUsedWithMeshDeformer=true
  bUsedWithMorphTargets=true
  bUsedWithParticleSprites=true
  bUsedWithSkeletalMesh=true
  bUsedWithSplineMeshes=true
  pixelDepthOffsetMode=PDOM_Legacy
  usedShadingModels=Default Lit
 }
 nodes [
  n1 = VectorParameter() {parameterName=WorldPostionOfBlackHole, defaultValue=<rgba -30000 25536 1952 0>}
  n2 = AppendVector(A=@n1.RGB, B=@n1.RGB)
  n3 = MaterialFunctionCall!(
   World Position=WorldPosition().XYZ
   DistanceCullFadeAnimationOffset=MaterialFunctionCall() {materialFunction=/Game/Packages/Fortress_SharedMaterials/Base_Material_Functions/PickupEffectDistanceCullOverride}.Result
   Use actor center instead of pivot point
  ) {materialFunction=/Game/Packages/Fortress_SharedMaterials/Base_Material_Functions/LODEffect}
  n4 = Add(A=@n3.Result, B=WorldPosition().XYZ)
  n5 = SetMaterialAttributes!(
   MaterialAttributes=MaterialFunctionCall!(
    Pre-Skinned World Normal
    Gmap
    Customization Mask
    IsCharacter
    IsStatsB
    Material Attributes=MaterialFunctionCall!(UV 1 coordinates, MuzzleFalloff) {materialFunction=/Game/Packages/Fortnite_UniqueMaterials_1/Weapons/M_GunMasterShaderFunction}.Material Result
   ) {materialFunction=/Game/Weapons/WeaponSkins/MaterialFunction/Weapon_SkinFXFunction}.Result
   World Position Offset=Add(
    A=MaterialFunctionCall!(WorldPostionOfBlackHole=@n2, World Position=@n4) {materialFunction=/Game/Packages/Fortnite_UniqueMaterials_1/BlackHoleFunction}.Result
    B=@n3.Result
   )
  ) {attributeSetTypes [#1]}
  n6 = TextureSampleParameter2D() {group=Skin FX, parameterName=SkinFX_Mask, samplerType=SAMPLERTYPE_Masks, texture=&T_BlackMask}
  MaterialFunctionCall!(In=MaterialFunctionCall!(In=MaterialFunctionCall!(Material=@n5, Alpha=@n6.R) {materialFunction=&/MF_LemonCart_Bomb}.Result) {materialFunction=/Game/Packages/Fortress_SharedMaterials/Base_Material_Functions/Hologram}.Material Att) {materialFunction=/Game/Effects/Niagara/MasterMaterials/Functions/MF_FX_EmissiveAdjustments_MA}
 ]
}
